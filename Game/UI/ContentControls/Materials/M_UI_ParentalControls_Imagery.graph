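# UI-domain translucent material for parental controls imagery that samples an Image texture parameter with adjustable framing.
# Normalized widget UVs are scaled about the center by Scale, shifted by a Pivot X/Y point, then passed to MF_UI_UVScaling.
# Its Scaled to Fit output gets an added rerouted value and Offset X/Y before driving a clamped texture sample.
# Named reroutes carry the texture size and pivot point.
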Material /Game/UI/ContentControls/Materials/M_UI_ParentalControls_Imagery
graph {
 properties {
  bUsedWithStaticMesh=false
  blendMode=BLEND_Translucent
  materialDomain=MD_UI
  pixelDepthOffsetMode=PDOM_Legacy
  usedShadingModels=Default Lit
 }
 nodes [
  n1 = TextureObjectParameter() {parameterName=Image}
  NamedRerouteDeclaration(None=TextureProperty(Tex=@n1)) {name=Texture Size, nodeColor=<rgba 0 0 0 1>}
  n2 = MaterialFunctionCall() {materialFunction=&GetUserInterfaceUV}
  n3 = ScalarParameter() {defaultValue=1, parameterName=Scale}
  n4 = AppendVector(A=ScalarParameter() {parameterName=Pivot X}, B=ScalarParameter() {parameterName=Pivot Y})
  n5 = MaterialFunctionCall(
   UV=Subtract(
    A=MaterialFunctionCall(UVs=@n2.Normalized UV, Texture Scale=@n3) {materialFunction=&ScaleUVsByCenter}.UVs
    B=NamedRerouteDeclaration(None=@n4) {name=Pivot Point, nodeColor=<rgba 0 0 0 1>}
   )
   TextureSize=NamedRerouteUsage()
  ) {materialFunction=&MF_UI_UVScaling}
  n6 = NamedRerouteUsage()
  n7 = ScalarParameter() {parameterName=Offset X}
  n8 = ScalarParameter() {parameterName=Offset Y}
  TextureSample(UVs=Add(A=Add(A=@n5.Scaled to Fit, B=@n6), B=AppendVector(A=@n7, B=@n8)), Tex=@n1) {samplerSource=SSM_Clamp_WorldGroupSettings}
 ]
}
