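# WBP_InputField is a reusable FortInputField text entry widget with optional label, error message and icon, processing throbber and gamepad Confirm glyph.
# PreConstruct applies label, hint, alignment and size options and creates the outline's dynamic material.
# Focus recolors outline and text, while hover or focus drives one guarded forward/reverse outline animation.
# SetErrorText toggles idle and error states with an icon shake, and input-device changes refresh glyph visibility.
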WidgetBlueprint /Game/UI/ContentControls/WBP_InputField
animations [
 HoverAnim {
  binding_count=1
  length_seconds=0.2
  object=&self:HoverAnim
  sequence {
   bindings [Outline {
     id=#1
     possessed_class=Image
     tracks [{
       type=MovieSceneWidgetMaterialTrack
       sections [{type=MovieSceneParameterSection, channels [Outline Size Ratio {type=MovieSceneScriptingFloatChannel, keys [[0,0.5],[12,1]]}]}]
      }]
    }]
   display_rate {numerator=60}
  }
 }
 ErrorAnim {
  binding_count=1
  length_seconds=0.3
  object=&self:ErrorAnim
  sequence {bindings [ErrorIcon {
     id=#2
     possessed_class=Image
     tracks [{
       type=MovieScene2DTransformTrack
       sections [{
         type=MovieScene2DTransformSection
         channels [
          Angle {type=MovieSceneScriptingFloatChannel, keys [[0,0],[2,-12],[4,12],[6,0]]}
          Scale.X {type=MovieSceneScriptingFloatChannel, keys [[0,1],[3,2],[6,1]]}
          Scale.Y {type=MovieSceneScriptingFloatChannel, keys [[0,1],[3,2],[6,1]]}
         ]
        }]
      }]
    }]}
 }
]
graphs [
 EventGraph {nodes [
   n1 = OnInputDeviceChanged!() -> (OutputDelegate, then, bNewInputType:ECommonInputType Enum) {is=title K2Node_CustomEvent}
   > SetActionVisibility()
   RefreshHoverAnim() {is=title K2Node_CustomEvent}
   n2 = Get p_IsMouseHovered() -> (p_IsMouseHovered:bool)
   > n3 = Branch(Condition=@n2)
   > n4 = PlaySound2D(Sound=/Game/Sounds/Fort_UI_Sounds/FrontEnd/New_FrontEnd/Fortnite_UI_MainMenu_Tab_Hover_01_Cue)
   n5 = OR Boolean(A=Get p_IsFocused() -> (p_IsFocused:bool), B=@n2)
   Branch(
    execute=[@n4.then, @n3.else]
    Condition="Not Equal (Boolean)"(A=Get p_FocusAnimPlayed() -> (p_FocusAnimPlayed:bool), B=@n5) -> (ReturnValue=~)
   )
   > n6 = Set p_FocusAnimPlayed(p_FocusAnimPlayed:bool=@n5) -> (Output_Get:bool)
   > n7 = Branch(Condition=@n6)
   n8 = Get HoverAnim() -> (HoverAnim:Widget Animation Object Reference)
   > PlayAnimationForward(InAnimation=@n8)
   PlayAnimationReverse(execute=@n7.else, InAnimation=@n8)
   OnError() {is=title K2Node_CustomEvent}
   > PlayAnimation(InAnimation=Get ErrorAnim() -> (ErrorAnim:Widget Animation Object Reference))
   > SetInputFieldState!(execute, self, State:E_UI_InputField_State Enum=NewEnumerator1) -> (then)
   n9 = Event PreConstruct()
   > n10 = GetDynamicMaterial(self=Get Outline() -> (Outline:Image Object Reference))
   > Set OutlineMaterial(OutlineMaterial:Material Instance Dynamic Object Reference=@n10) -> (Output_Get:Material Instance Dynamic Object Reference)
   > SetLabelDisplay()
   > SetHintText!(
    execute
    self:Editable Text Object Reference=Get EditableTextWidget() -> (EditableTextWidget:Editable Text Object Reference)
    InHintText:Text=Get DefaultHintTextValue() -> (DefaultHintTextValue:Text)
   ) -> (then)
   > SetContentAlignment()
   > SetLabelValue!(execute, self, LabelValue:Text=Get DefaultLabelValue() -> (DefaultLabelValue:Text)) -> (then)
   > SetContentSize()
   > Set Focus Visual!(execute, self, IsFocused:bool=false) -> (then) {is=title K2Node_CallFunction}
   > Branch(Condition=@n9.IsDesignTime)
   > SetInputFieldState(State=Get DebugInputFieldState() -> (DebugInputFieldState:E_UI_InputField_State Enum=NewEnumerator0))
   Event OnAddedToFocusPath!() -> (OutputDelegate, then, InFocusEvent:Focus Event Structure) {is=title K2Node_Event}
   > n11 = Set p_IsFocused(p_IsFocused:bool=true) -> (Output_Get:bool)
   > n12 = Set Focus Visual(IsFocused=@n11)
   RefreshHoverAnim(execute=[@n15.then, @n12.then]) {is=title K2Node_CallFunction}
   Event Destruct()
   n13 = Get CommonInputSubsystem()
   > Unbind Event from On Input Method Changed(self=@n13, Delegate=@n1)
   Event OnRemovedFromFocusPath!() -> (OutputDelegate, then, InFocusEvent:Focus Event Structure) {is=title K2Node_Event}
   > n14 = Set p_IsFocused(p_IsFocused=false)
   > n15 = Set Focus Visual(IsFocused=@n14)
   Event OnMouseEnter()
   > n16 = Set p_IsMouseHovered(p_IsMouseHovered:bool=true) -> (Output_Get:bool)
   RefreshHoverAnim(execute=[@n17.then, @n16.then])
   Event OnMouseLeave!() -> (OutputDelegate, then, MouseEvent:"Pointer Event Structure (by ref)") {is=title K2Node_Event}
   > n17 = Set p_IsMouseHovered(p_IsMouseHovered=false)
   n18 = Event SetEditableText!() -> (OutputDelegate, then, EditableText:"Text (by ref)") {is=title K2Node_Event}
   > "SetText (Editable Text)"!(execute, self:Editable Text Object Reference=Get EditableTextWidget(), InText:Text=@n18.EditableText) -> (then)
   n19 = Event SetErrorText!() -> (OutputDelegate, then, ErrorText:"Text (by ref)") {is=title K2Node_Event}
   > Set Text(Text=@n19.ErrorText, self=Get ErrorText() -> (ErrorText:Common Text Object Reference))
   > n20 = Branch(Condition=TextIsEmpty(InText=@n19.ErrorText))
   > SetInputFieldState(State=NewEnumerator0)
   OnError(execute=@n20.else) {is=title K2Node_CallFunction}
   n21 = Event SetHintText!() -> (OutputDelegate, then, HintText:"Text (by ref)") {is=title K2Node_Event}
   > SetHintText(self=Get EditableTextWidget(), InHintText=@n21.HintText)
   Event OnInitialized()
   > SetInputFieldState(State=NewEnumerator0)
   Event Construct()
   > Bind Event to On Input Method Changed(self=@n13, Delegate=@n1)
  ]}
 OnFocusReceived {nodes [
   OnFocusReceived!() -> (then, MyGeometry:Geometry Structure, InFocusEvent:Focus Event Structure) {is=title K2Node_FunctionEntry}
   > SetFocus(self=Get EditableTextWidget())
   > Return Node!(execute, ReturnValue:Event Reply Structure=Make Event Reply!() -> (EventReply:Event Reply Structure) {is=title K2Node_MakeStruct})
  ]}
 OnMouseButtonUp {nodes [
   OnMouseButtonUp!() -> (then, MyGeometry:Geometry Structure, MouseEvent:"Pointer Event Structure (by ref)") {is=title K2Node_FunctionEntry}
   > SetFocus(self=Get EditableTextWidget())
   > Return Node(ReturnValue=Make Event Reply())
  ]}
 OnTouchStarted {nodes [
   OnTouchStarted!() -> (then, MyGeometry:Geometry Structure, InTouchEvent:"Pointer Event Structure (by ref)") {is=title K2Node_FunctionEntry}
   > SetFocus(self=Get EditableTextWidget())
   > Return Node(ReturnValue=Make Event Reply())
  ]}
 Set Focus Visual {nodes [
   n1 = Set Focus Visual!() -> (then, IsFocused:bool=~) {is=title K2Node_FunctionEntry}
   > SetVectorParameterValue(
    self=Get OutlineMaterial() -> (OutlineMaterial:Material Instance Dynamic Object Reference)
    ParameterName=Get OutlineColorParam() -> (OutlineColorParam:Name=None)
    Value=SelectColor(A=Get OutlineEndColor() -> (OutlineEndColor:Linear Color Structure), B=(R=0,G=0.0908,B=0.9911,A=1), bPickA=@n1)
   )
   n2 = Break Editable Text Style!(EditableTextStyle:"Editable Text Style Structure (by ref)"=Get WidgetStyle!(self:Editable Text Object Reference=Get EditableTextWidget()) -> (WidgetStyle:Editable Text Style Structure)) -> (
    Font:Slate Font Info Structure
    ColorAndOpacity:Slate Color Structure
    HintColorAndOpacity:Slate Color Structure
    BackgroundImageSelected:Slate Brush Structure
    SelectedBackgroundColor:Slate Color Structure
    BackgroundImageComposing:Slate Brush Structure
    CaretImage:Slate Brush Structure
   ) {is=title K2Node_BreakStruct}
   n3 = SelectColor(A=(R=1,G=1,B=1,A=1), B=(R=0.6038,G=1,B=1,A=1), bPickA=Get IsFocused() -> (IsFocused:bool))
   n4 = Get EditableTextWidget()
   > Set WidgetStyle!(
    execute
    WidgetStyle:Editable Text Style Structure=Make Editable Text Style!(
     Font:Slate Font Info Structure=@n2
     ColorAndOpacity:Slate Color Structure=Make Slate Color(SpecifiedColor=@n3)
     HintColorAndOpacity:Slate Color Structure=(SpecifiedColor=(R=1,G=0,B=1,A=1),ColorUseRule=UseColor_Foreground)
     BackgroundImageSelected:Slate Brush Structure=@n2.BackgroundImageSelected
     SelectedBackgroundColor:Slate Color Structure=(SpecifiedColor=(R=1,G=0,B=1,A=1),ColorUseRule=UseColor_Foreground)
     BackgroundImageComposing:Slate Brush Structure=@n2.BackgroundImageComposing
     CaretImage:Slate Brush Structure=@n2.CaretImage
    ) -> (EditableTextStyle:Editable Text Style Structure) {is=title K2Node_MakeStruct}
    self:Editable Text Object Reference=@n4
   ) -> (then, Output_Get:Editable Text Style Structure)
   > SetFont(
    self:Editable Text Object Reference=@n4
    InFontInfo=GetFont!(self:Editable Text Object Reference=@n4) -> (ReturnValue:"Slate Font Info Structure (by ref)")
   )
   > SetActionVisibility()
  ]}
 SetActionVisibility {nodes [
   SetActionVisibility() {is=title K2Node_FunctionEntry}
   n1 = OR Boolean(A=NOT Boolean(A=Get ShowActionOnlyOnFocus() -> (ShowActionOnlyOnFocus:bool)), B=Get p_IsFocused())
   n2 = "Equal (Enum)"(A=GetCurrentInputType(self=Get CommonInputSubsystem()), B=Gamepad)
   > SetVisibility(
    self=Get Overlay_ActionWidgetShower() -> (Overlay_ActionWidgetShower:Overlay Object Reference)
    InVisibility=Select(Option 0=Collapsed, Option 1=HitTestInvisible, Index=AND Boolean(A=@n1, B=@n2))
   )
  ]}
 SetContentAlignment {nodes [
   SetContentAlignment() {is=title K2Node_FunctionEntry}
   > Set Justification!(
    execute
    Justification:ETextJustify Enum=Select(Option 0:ETextJustify Enum=Left, Option 1:ETextJustify Enum=Center, Index=Get AlignContentCenter() -> (AlignContentCenter:bool)) -> (ReturnValue:ETextJustify Enum)
    self:Editable Text Object Reference=Get EditableTextWidget()
   ) -> (then, Output_Get:ETextJustify Enum=Left)
  ]}
 SetContentSize {nodes [
   SetContentSize() {is=title K2Node_FunctionEntry}
   n1 = Break Slate Font Info!(SlateFontInfo:"Slate Font Info Structure (by ref)"=GetFont(self=Get EditableTextWidget())) -> (
    FontObject:Object Reference
    FontMaterial:Object Reference
    OutlineSettings:Font Outline Settings Structure
    TypefaceFontName:Name
    Size:float=~
    LetterSpacing:int
    SkewAmount:float=~
    bForceMonospaced:bool=~
    bMaterialIsStencil:bool=~
    MonospacedWidth:float=~
   ) {is=title K2Node_BreakStruct}
   n2 = Get ContentSize() -> (ContentSize:E_UI_InputField_ContentSize Enum=NewEnumerator0)
   n3 = Select!(NewEnumerator0:int=21, NewEnumerator1:int=24, NewEnumerator2:int=36, Index:E_UI_InputField_ContentSize Enum=@n2) -> (ReturnValue:int)
   > SetFont(
    self=Get EditableTextWidget()
    InFontInfo=MakeSlateFontInfo!(
     FontObject:Object Reference=Select(
      NewEnumerator0:Object Reference=&BurbankSmall
      NewEnumerator1:Object Reference=&BurbankSmall
      NewEnumerator2:Object Reference=&Burbank
      Index=@n2
     ) -> (ReturnValue:Object Reference)
     FontMaterial:Object Reference=@n1.FontMaterial
     OutlineSettings:Font Outline Settings Structure=@n1.OutlineSettings
     TypefaceFontName:Name=Select(NewEnumerator0:Name=Bold, NewEnumerator1:Name=Bold, NewEnumerator2:Name=BurbankRegular-Black, Index=@n2) -> (ReturnValue:Name)
     Size:float="To Float (Integer)"(InInt=@n3)
     LetterSpacing:int=@n1.LetterSpacing
     SkewAmount:float=@n1.SkewAmount
     bForceMonospaced:bool=false
     bMaterialIsStencil:bool=false
     MonospacedWidth:float=1
    ) -> (ReturnValue:Slate Font Info Structure)
   )
  ]}
 SetInputFieldState {nodes [
   SetInputFieldState!() -> (then, State:E_UI_InputField_State Enum=NewEnumerator0) {is=title K2Node_FunctionEntry}
   > SetVisibility(
    self=Get ErrorTextWrapper() -> (ErrorTextWrapper:Border Object Reference)
    InVisibility=Select(
     NewEnumerator0:ESlateVisibility Enum=Collapsed
     NewEnumerator1:ESlateVisibility Enum=SelfHitTestInvisible
     NewEnumerator2:ESlateVisibility Enum=Collapsed
     Index:E_UI_InputField_State Enum=Get State() -> (State:E_UI_InputField_State Enum=NewEnumerator0)
    ) -> (ReturnValue:ESlateVisibility Enum)
   )
   > SetVisibility(
    self=Get ErrorIconWrapper() -> (ErrorIconWrapper:Overlay Object Reference)
    InVisibility=Select(NewEnumerator0=Hidden, NewEnumerator1=SelfHitTestInvisible, NewEnumerator2=Hidden, Index=Get State() -> (State=NewEnumerator0))
   )
   > Set OutlineEndColor(OutlineEndColor:Linear Color Structure=Select(
     NewEnumerator0:Linear Color Structure=Get IdleEndColor() -> (IdleEndColor:Linear Color Structure)
     NewEnumerator1:Linear Color Structure=Get ErrorEndColor() -> (ErrorEndColor:Linear Color Structure)
     NewEnumerator2:Linear Color Structure
     Index=Get State() -> (State=NewEnumerator0)
    ) -> (ReturnValue:Linear Color Structure)) -> (Output_Get:Linear Color Structure)
  ]}
 SetLabelDisplay {nodes [
   SetLabelDisplay() {is=title K2Node_FunctionEntry}
   > SetVisibility(
    self=Get LabelTextWrapper() -> (LabelTextWrapper:Border Object Reference)
    InVisibility=Select!(
     Option 0:ESlateVisibility Enum=Collapsed
     Option 1:ESlateVisibility Enum=SelfHitTestInvisible
     Index:bool=Get ShowLabel() -> (ShowLabel:bool)
    ) -> (ReturnValue:ESlateVisibility Enum)
   )
  ]}
 SetLabelValue {nodes [
   n1 = SetLabelValue!() -> (then, LabelValue:Text) {is=title K2Node_FunctionEntry}
   > "SetText (Text)"(self=Get LabelText() -> (LabelText:Common Text Object Reference), InText=@n1)
  ]}
 SetProcessingVisibility {nodes [
   n1 = SetProcessingVisibility!() -> (then, IsVisible:bool=~) {is=title K2Node_FunctionEntry}
   > SetVisibility(
    self=Get ProcessingFeedbackWrapper() -> (ProcessingFeedbackWrapper:Overlay Object Reference)
    InVisibility=Select(Option 0=Collapsed, Option 1=HitTestInvisible, Index=@n1)
   )
  ]}
 SetTextValue {nodes [
   n1 = SetTextValue!() -> (then, TextValue:Text) {is=title K2Node_FunctionEntry}
   > "SetText (Editable Text)"(self=Get EditableTextWidget(), InText=@n1)
  ]}
]
variables [
 AlignContentCenter {type=boolean}
 ContentSize {type=E_UI_InputField_ContentSize, default=Small}
 DebugInputFieldState {type=E_UI_InputField_State, default=Idle}
 DefaultHintTextValue {type=string, default=Hint Text}
 DefaultLabelValue {type=string, default=Label}
 ErrorEndColor {type=LinearColor, default=<rgba 1 0.1714 0.1329 1>}
 ErrorValue {type=string, default=Error Text Value}
 IdleEndColor {type=LinearColor, default=<rgba 0.0331 0.7011 1 1>}
 OutlineColorParam {type=string, default=Outline Color}
 OutlineEndColor {type=LinearColor, default=<rgba 0.0331 0.7011 1 1>}
 OutlineMaterial {type=MaterialInstanceDynamic}
 p_FocusAnimPlayed {type=boolean}
 p_IsFocused {type=boolean}
 p_IsMouseHovered {type=boolean}
 ShowActionOnlyOnFocus {type=boolean}
 ShowLabel {type=boolean}
]
widgets {
 info {inheritedWidgetCount=1, parentClass=/Script/FortniteUI.FortInputField, rootWidgetClass=&Overlay, widgetCount=19}
 widgets [
  Overlay_0 {widgetClassPath=&Overlay, widget=Overlay_0}
  VerticalBox_38 {
   widgetClassPath=&VerticalBox
   parent=Overlay_0
   widget=VerticalBox_38
   slot {class=OverlaySlot, properties {horizontalAlignment=HAlign_Fill}}
  }
  LabelTextWrapper {
   widgetClassPath=&Border
   bIsVariable=true
   parent=VerticalBox_38
   widget=LabelTextWrapper
   properties {
    visibility=HitTestInvisible
    background {
     bIsDynamicallyLoaded=false
     drawAs=NoDrawType
     imageType=NoImage
     mirroring=NoMirror
     resourceName=None
     resourceObject=null
     tiling=NoTile
     imageSize=<32 32>
     margin=<ltrb 0 0 0 0>
     outlineSettings {
      bUseBrushTransparency=false
      roundingType=HalfHeightRadius
      width=0
      color {colorUseRule=UseColor_Specified, specifiedColor=<rgba 0 0 0 0>}
      cornerRadii=<0 0 0 0>
     }
     tintColor {colorUseRule=UseColor_Specified, specifiedColor=<rgba 1 1 1 1>}
     uVRegion {bIsValid=false, max=<0 0>, min=<0 0>}
    }
    contentColorAndOpacity=<rgba 0.6038 1 1 1>
    padding=<ltrb 0 0 0 0>
   }
   slot {class=VerticalBoxSlot, properties {padding=<ltrb 0 0 0 8>}}
  }
  LabelText {
   widgetClassPath=&CommonTextBlock
   bIsVariable=true
   parent=LabelTextWrapper
   widget=LabelText
   properties {
    autoWrapText=true
    lineHeightPercentage=1.29
    mobileFontSizeMultiplier=1.3
    style=/FortUILibrary/Styles/Typography/_TextStyle_Body1._TextStyle_Body1_C
    text=Label
    font {
     bForceMonospaced=false
     bMaterialIsStencil=false
     fontMaterial=null
     fontObject=&BurbankSmall
     letterSpacing=0
     monospacedWidth=1
     size=18
     skewAmount=0
     typefaceFontName=Bold
     outlineSettings {
      bApplyOutlineToDropShadows=false
      bMiteredCorners=false
      bSeparateFillAlpha=false
      outlineMaterial=null
      outlineSize=0
      outlineColor=<rgba 0 0 0 1>
     }
    }
    shadowOffset=<0 0>
    strikeBrush {
     bIsDynamicallyLoaded=false
     drawAs=NoDrawType
     imageType=NoImage
     mirroring=NoMirror
     resourceName=None
     resourceObject=null
     tiling=NoTile
     imageSize=<32 32>
     margin=<ltrb 0 0 0 0>
     outlineSettings {
      bUseBrushTransparency=false
      roundingType=HalfHeightRadius
      width=0
      color {colorUseRule=UseColor_Specified, specifiedColor=<rgba 0 0 0 0>}
      cornerRadii=<0 0 0 0>
     }
     tintColor {colorUseRule=UseColor_Specified, specifiedColor=<rgba 1 1 1 1>}
     uVRegion {bIsValid=false, max=<0 0>, min=<0 0>}
    }
   }
   slot {class=BorderSlot, properties {padding=<ltrb 0 0 0 0>}}
  }
  InputFieldWrapper {
   widgetClassPath=&Border
   parent=VerticalBox_38
   widget=InputFieldWrapper
   properties {visibility=SelfHitTestInvisible, brushColor=<rgba 0.0024 0.016 0.117 1>, padding=<ltrb 0 0 0 0>}
   slot {class=VerticalBoxSlot}
  }
  Overlay_3 {
   widgetClassPath=&Overlay
   parent=InputFieldWrapper
   widget=Overlay_3
   slot {class=BorderSlot, properties {padding=<ltrb 0 0 0 0>}}
  }
  Outline {
   widgetClassPath=&Image
   bIsVariable=true
   parent=Overlay_3
   widget=Outline
   properties {brush {
     bIsDynamicallyLoaded=false
     drawAs=Box
     imageType=NoImage
     mirroring=NoMirror
     resourceName=None
     resourceObject=/Game/UI/ContentControls/MaterialInstances/MI_UI_ParentalControls_Input_Border
     tiling=NoTile
     imageSize=<32 32>
     margin=<ltrb 0.5 0.5 0.5 0.5>
     outlineSettings {
      bUseBrushTransparency=false
      roundingType=HalfHeightRadius
      width=0
      color {colorUseRule=UseColor_Specified, specifiedColor=<rgba 0 0 0 0>}
      cornerRadii=<0 0 0 0>
     }
     tintColor {colorUseRule=UseColor_Specified, specifiedColor=<rgba 1 1 1 1>}
     uVRegion {bIsValid=false, max=<0 0>, min=<0 0>}
    }}
   slot {class=OverlaySlot, properties {horizontalAlignment=HAlign_Fill, verticalAlignment=VAlign_Fill}}
  }
  HorizontalBox_0 {
   widgetClassPath=&HorizontalBox
   parent=Overlay_3
   widget=HorizontalBox_0
   slot {class=OverlaySlot, properties {horizontalAlignment=HAlign_Fill, verticalAlignment=VAlign_Fill, padding=<ltrb 16 16 16 16>}}
  }
  Overlay_ActionWidgetShower {
   widgetClassPath=&Overlay
   bIsVariable=true
   parent=HorizontalBox_0
   widget=Overlay_ActionWidgetShower
   properties {visibility=Collapsed}
   slot {class=HorizontalBoxSlot, properties {horizontalAlignment=HAlign_Left, verticalAlignment=VAlign_Center, padding=<ltrb 0 0 8 0>}}
  }
  CommonActionWidget {
   widgetClassPath=&CommonActionWidget
   bIsVariable=true
   parent=Overlay_ActionWidgetShower
   widget=CommonActionWidget
   properties {inputActions [{dataTable=&InputActionDataTable, rowName=Confirm}]}
   slot {class=OverlaySlot}
  }
  SizeBox_79 {
   widgetClassPath=&SizeBox
   parent=Overlay_ActionWidgetShower
   widget=SizeBox_79
   properties {bOverride_HeightOverride=true, bOverride_WidthOverride=true, heightOverride=32, widthOverride=32}
   slot {class=OverlaySlot}
  }
  EditableTextWidget {
   widgetClassPath=/Script/UMG.EditableText
   bInherited=true
   bIsVariable=true
   parent=HorizontalBox_0
   widget=EditableTextWidget
   properties {
    hintText=Enter Email
    justification=Center
    widgetStyle {
     backgroundImageComposing {
      bIsDynamicallyLoaded=false
      drawAs=Image
      imageType=FullColor
      mirroring=NoMirror
      resourceName=../../../Engine/Content/Slate/Old/HyperlinkDotted.png
      resourceObject=null
      tiling=NoTile
      imageSize=<0 0>
      margin=<ltrb 0 0 0 0.1875>
      outlineSettings {
       bUseBrushTransparency=false
       roundingType=HalfHeightRadius
       width=0
       color {colorUseRule=UseColor_Specified, specifiedColor=<rgba 0 0 0 0>}
       cornerRadii=<0 0 0 0>
      }
      tintColor {colorUseRule=UseColor_Specified, specifiedColor=<rgba 1 1 1 1>}
      uVRegion {bIsValid=false, max=<0 0>, min=<0 0>}
     }
     backgroundImageSelected {
      bIsDynamicallyLoaded=false
      drawAs=Box
      imageType=FullColor
      mirroring=NoMirror
      resourceName=../../../Engine/Content/Slate/Common/EditableTextSelectionBackground.png
      resourceObject=null
      tiling=NoTile
      imageSize=<0 0>
      margin=<ltrb 0.25 0.25 0.25 0.25>
      outlineSettings {
       bUseBrushTransparency=false
       roundingType=HalfHeightRadius
       width=0
       color {colorUseRule=UseColor_Specified, specifiedColor=<rgba 0 0 0 0>}
       cornerRadii=<0 0 0 0>
      }
      tintColor {colorUseRule=UseColor_Specified, specifiedColor=<rgba 0.0331 0.7011 1 1>}
      uVRegion {bIsValid=false, max=<0 0>, min=<0 0>}
     }
     caretImage {
      bIsDynamicallyLoaded=false
      drawAs=Image
      imageType=FullColor
      mirroring=NoMirror
      resourceName=../../../Engine/Content/Slate/Old/White.png
      resourceObject=null
      tiling=NoTile
      imageSize=<16 16>
      margin=<ltrb 0 0 0 0>
      outlineSettings {
       bUseBrushTransparency=false
       roundingType=HalfHeightRadius
       width=0
       color {colorUseRule=UseColor_Specified, specifiedColor=<rgba 0 0 0 0>}
       cornerRadii=<0 0 0 0>
      }
      tintColor {colorUseRule=UseColor_Specified, specifiedColor=<rgba 1 1 1 1>}
      uVRegion {bIsValid=false, max=<0 0>, min=<0 0>}
     }
     colorAndOpacity {colorUseRule=UseColor_Specified, specifiedColor=<rgba 0.6038 1 1 1>}
     font {
      bForceMonospaced=false
      bMaterialIsStencil=false
      fontMaterial=null
      fontObject=&BurbankSmall
      letterSpacing=0
      monospacedWidth=1
      size=21
      skewAmount=0
      typefaceFontName=Bold
      outlineSettings {
       bApplyOutlineToDropShadows=false
       bMiteredCorners=false
       bSeparateFillAlpha=false
       outlineMaterial=null
       outlineSize=0
       outlineColor=<rgba 0 0 0 1>
      }
     }
     hintColorAndOpacity {colorUseRule=UseColor_Foreground, specifiedColor=<rgba 1 0 1 1>}
     selectedBackgroundColor {colorUseRule=UseColor_Foreground, specifiedColor=<rgba 1 0 1 1>}
    }
   }
   slot {class=HorizontalBoxSlot, properties {verticalAlignment=VAlign_Center, size {sizeRule=Fill, value=1}}}
  }
  HorizontalBox_1 {
   widgetClassPath=&HorizontalBox
   parent=HorizontalBox_0
   widget=HorizontalBox_1
   slot {class=HorizontalBoxSlot, properties {verticalAlignment=VAlign_Center, padding=<ltrb 8 0 0 0>}}
  }
  ErrorIconWrapper {
   widgetClassPath=&Overlay
   bIsVariable=true
   parent=HorizontalBox_1
   widget=ErrorIconWrapper
   properties {visibility=Visible}
   slot {class=HorizontalBoxSlot}
  }
  ErrorIcon {
   widgetClassPath=&Image
   parent=ErrorIconWrapper
   widget=ErrorIcon
   properties {
    visibility=SelfHitTestInvisible
    brush {
     bIsDynamicallyLoaded=false
     drawAs=Image
     imageType=NoImage
     mirroring=NoMirror
     resourceName=None
     resourceObject=/Game/UI/ContentControls/Textures/T_UI_Icon_Warning
     tiling=NoTile
     imageSize=<32 32>
     margin=<ltrb 0 0 0 0>
     outlineSettings {
      bUseBrushTransparency=false
      roundingType=HalfHeightRadius
      width=0
      color {colorUseRule=UseColor_Specified, specifiedColor=<rgba 0 0 0 0>}
      cornerRadii=<0 0 0 0>
     }
     tintColor {colorUseRule=UseColor_Specified, specifiedColor=<rgba 1 1 1 1>}
     uVRegion {bIsValid=false, max=<0 0>, min=<0 0>}
    }
    colorAndOpacity=<rgba 1 0.1714 0.1329 1>
   }
   slot {class=OverlaySlot, properties {horizontalAlignment=HAlign_Right, verticalAlignment=VAlign_Center}}
  }
  ProcessingFeedbackWrapper {
   widgetClassPath=&Overlay
   bIsVariable=true
   parent=HorizontalBox_1
   widget=ProcessingFeedbackWrapper
   properties {visibility=Collapsed}
   slot {class=HorizontalBoxSlot, properties {horizontalAlignment=HAlign_Center, verticalAlignment=VAlign_Center}}
  }
  ProcessingIcon {
   widgetClassPath=&Image
   parent=ProcessingFeedbackWrapper
   widget=ProcessingIcon
   properties {
    visibility=SelfHitTestInvisible
    brush {
     bIsDynamicallyLoaded=false
     drawAs=Image
     imageType=FullColor
     mirroring=NoMirror
     resourceName=None
     resourceObject=/FortUILibrary/UIKit/Throbber/MI_UIKit_Color_Throbber
     tiling=NoTile
     imageSize=<32 32>
     margin=<ltrb 0 0 0 0>
     outlineSettings {
      bUseBrushTransparency=false
      roundingType=HalfHeightRadius
      width=0
      color {colorUseRule=UseColor_Specified, specifiedColor=<rgba 0 0 0 0>}
      cornerRadii=<0 0 0 0>
     }
     tintColor {colorUseRule=UseColor_Specified, specifiedColor=<rgba 1 1 1 1>}
     uVRegion {bIsValid=false, max=<0 0>, min=<0 0>}
    }
   }
   slot {class=OverlaySlot}
  }
  ErrorTextWrapper {
   widgetClassPath=&Border
   bIsVariable=true
   parent=VerticalBox_38
   widget=ErrorTextWrapper
   properties {
    visibility=Hidden
    background {
     bIsDynamicallyLoaded=false
     drawAs=NoDrawType
     imageType=NoImage
     mirroring=NoMirror
     resourceName=None
     resourceObject=null
     tiling=NoTile
     imageSize=<32 32>
     margin=<ltrb 0 0 0 0>
     outlineSettings {
      bUseBrushTransparency=false
      roundingType=HalfHeightRadius
      width=0
      color {colorUseRule=UseColor_Specified, specifiedColor=<rgba 0 0 0 0>}
      cornerRadii=<0 0 0 0>
     }
     tintColor {colorUseRule=UseColor_Specified, specifiedColor=<rgba 1 1 1 1>}
     uVRegion {bIsValid=false, max=<0 0>, min=<0 0>}
    }
    contentColorAndOpacity=<rgba 1 0.1714 0.1329 1>
    padding=<ltrb 0 0 0 0>
   }
   slot {class=VerticalBoxSlot, properties {padding=<ltrb 0 8 0 0>}}
  }
  ErrorText {
   widgetClassPath=&CommonTextBlock
   bIsVariable=true
   parent=ErrorTextWrapper
   widget=ErrorText
   properties {
    autoWrapText=true
    lineHeightPercentage=1.29
    mobileFontSizeMultiplier=1.3
    style=&_TextStyle_Body1_C
    text=Error Text
    font {
     bForceMonospaced=false
     bMaterialIsStencil=false
     fontMaterial=null
     fontObject=&BurbankSmall
     letterSpacing=0
     monospacedWidth=1
     size=18
     skewAmount=0
     typefaceFontName=Bold
     outlineSettings {
      bApplyOutlineToDropShadows=false
      bMiteredCorners=false
      bSeparateFillAlpha=false
      outlineMaterial=null
      outlineSize=0
      outlineColor=<rgba 0 0 0 1>
     }
    }
    shadowOffset=<0 0>
    strikeBrush {
     bIsDynamicallyLoaded=false
     drawAs=NoDrawType
     imageType=NoImage
     mirroring=NoMirror
     resourceName=None
     resourceObject=null
     tiling=NoTile
     imageSize=<32 32>
     margin=<ltrb 0 0 0 0>
     outlineSettings {
      bUseBrushTransparency=false
      roundingType=HalfHeightRadius
      width=0
      color {colorUseRule=UseColor_Specified, specifiedColor=<rgba 0 0 0 0>}
      cornerRadii=<0 0 0 0>
     }
     tintColor {colorUseRule=UseColor_Specified, specifiedColor=<rgba 1 1 1 1>}
     uVRegion {bIsValid=false, max=<0 0>, min=<0 0>}
    }
   }
   slot {class=BorderSlot, properties {padding=<ltrb 0 0 0 0>}}
  }
 ]
}
